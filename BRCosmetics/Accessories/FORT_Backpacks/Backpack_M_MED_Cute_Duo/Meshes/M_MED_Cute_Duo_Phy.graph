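# Physics asset for the Cute Duo backpack cosmetic, giving its dangling masks, straps and bow simple collision bodies and joints for secondary motion.
# The pack body combines two boxes and a capsule, anchoring a four-link mask chain, two straps and the bowknot base.
# The bowknot base carries the flower, two knots and two bow straps.
# Joints set per-part swing and twist limits, but the right strap's angular motion is explicitly free, unlike the left.
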PhysicsAsset /BRCosmetics/Accessories/FORT_Backpacks/Backpack_M_MED_Cute_Duo/Meshes/M_MED_Cute_Duo_Phy
bodies {
 dyn_bowStrap_a_01 [{shapeType=Capsule, length=1.5, radius=1, center=<2 0 0>, rotation=<rot 0 0 90>}]
 dyn_bowStrap_b_01 [{shapeType=Capsule, length=1.5, radius=1, center=<2 0 0>, rotation=<rot 0 0 90>}]
 dyn_bowknot_base [{shapeType=Capsule, length=1, radius=1}]
 dyn_flower [{shapeType=Capsule, length=3, radius=0.5, center=<1 0 0>, rotation=<rot 90 0 0>}]
 dyn_knot_01 [{shapeType=Capsule, length=2, radius=1, center=<1.2 0 0>, rotation=<rot 90 0 0>}]
 dyn_knot_02 [{shapeType=Capsule, length=2, radius=1, center=<1 0 0>, rotation=<rot 90 0 0>}]
 dyn_masks_01 [{shapeType=Box, extentX=5.0133, extentY=2.2921, extentZ=0.943, center=<2.2092 0.0001 0>}]
 dyn_masks_02 [{shapeType=Box, extentX=6, extentY=7.5, extentZ=2, center=<3 0 0>}]
 dyn_masks_03 [{shapeType=Box, extentX=6, extentY=7.5, extentZ=2, center=<3 0 0>}]
 dyn_masks_04 [{shapeType=Capsule, length=2.5, radius=0.8, center=<2 0 0>, rotation=<rot 90 0 0>}]
 dyn_straps_l [{shapeType=Capsule, length=1, radius=1, center=<4 0 0.4>, rotation=<rot 0 0 90>}]
 dyn_straps_r [{shapeType=Capsule, length=1, radius=1, center=<4 0 0.4>, rotation=<rot 0 0 90>}]
 pack [
  {shapeType=Capsule, length=3.25, radius=1.5, center=<11.7482 -4.7173 -13.122>, rotation=<rot 8.6493 -5.0387 149.6182>}
  {shapeType=Box, extentX=21.5, extentY=14, extentZ=30, center=<0 -5.8818 -7.1522>}
  {shapeType=Box, extentX=38, extentY=3, extentZ=35, center=<-9.1887 3.1352 -6.6957>}
 ]
}
constraints [
 {bone1Name=dyn_masks_01, bone2Name=pack, swing1LimitDegrees=20, swing2LimitDegrees=15, twistLimitDegrees=2}
 {bone1Name=dyn_masks_02, bone2Name=dyn_masks_01, swing1LimitDegrees=30, swing2LimitDegrees=5, twistLimitDegrees=5}
 {bone1Name=dyn_masks_03, bone2Name=dyn_masks_02, swing1LimitDegrees=30, swing2LimitDegrees=5, twistLimitDegrees=5}
 {bone1Name=dyn_masks_04, bone2Name=dyn_masks_03, swing1LimitDegrees=30, swing2LimitDegrees=10, twistLimitDegrees=10}
 {bone1Name=dyn_straps_l, bone2Name=pack, swing1LimitDegrees=60, swing2LimitDegrees=90, twistLimitDegrees=8}
 {
  bone1Name=dyn_straps_r
  bone2Name=pack
  swing1LimitDegrees=60
  swing1Motion=Free
  swing2LimitDegrees=90
  swing2Motion=Free
  twistLimitDegrees=8
  twistMotion=Free
 }
 {bone1Name=dyn_bowknot_base, bone2Name=pack, swing1LimitDegrees=15, swing2LimitDegrees=15, twistLimitDegrees=5}
 {bone1Name=dyn_flower, bone2Name=dyn_bowknot_base, swing1LimitDegrees=10, swing2LimitDegrees=10, twistLimitDegrees=10}
 {bone1Name=dyn_knot_01, bone2Name=dyn_bowknot_base, swing1LimitDegrees=30, swing2LimitDegrees=15, twistLimitDegrees=5}
 {bone1Name=dyn_knot_02, bone2Name=dyn_bowknot_base, swing1LimitDegrees=30, swing2LimitDegrees=15, twistLimitDegrees=5}
 {bone1Name=dyn_bowStrap_a_01, bone2Name=dyn_bowknot_base, swing1LimitDegrees=45, swing2LimitDegrees=45, twistLimitDegrees=5}
 {bone1Name=dyn_bowStrap_b_01, bone2Name=dyn_bowknot_base, swing1LimitDegrees=45, swing2LimitDegrees=45, twistLimitDegrees=5}
]
